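# Physics asset for the WormChalkUniversal two-handed axe pickaxe, a cosmetic melee weapon.
# It holds a single capsule body on the Melee_TwoHanded_Handle bone, oriented and offset to run along the weapon's length.
# There are no constraints, so the whole weapon simulates and collides as one rigid piece.
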PhysicsAsset /BRCosmetics/Weapons/FORT_Melee/Pickaxe_wormChalkUniversal/Meshes/WormChalkUniversal_Axe_PhysicsAsset
bodies {Melee_TwoHanded_Handle [{shapeType=Capsule, length=56.8124, radius=6.6188, center=<0.0051 0.0338 33.5427>, rotation=<rot 0.1799 -90 0.0219>}]}
constraints []
